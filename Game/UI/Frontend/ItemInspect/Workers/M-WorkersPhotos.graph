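# Two-sided lit material for worker photos in the frontend item inspect UI, showing a portrait inside a tinted frame.
# The portrait can be flipped, scaled, offset and LUT-graded, then composited by alpha over an RGBA0/RGBA1 vertical gradient unless IgnoreRarity is set.
# A procedural square UV mask blends it with the frame color, and TwoSidedSign limits the portrait to front faces.
# Tiled generic frame textures supply roughness variation and normal detail.
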Material /Game/UI/Frontend/ItemInspect/Workers/M-WorkersPhotos
graph {
 properties {bUsedWithMeshDeformer=true, bUsedWithSkeletalMesh=true, pixelDepthOffsetMode=PDOM_Legacy, twoSided=true, usedShadingModels=Default Lit}
 nodes [
  Constant()
  n1 = TwoSidedSign()
  LinearInterpolate(Alpha=@n1)
  n2 = Constant3Vector() {constant=<rgba 0.84 0.8349 0.8274 0>}
  n3 = Constant3Vector() {constant=<rgba 0.75 0.6833 0.5625 0>}
  n4 = Add(A=Rotator(Coordinate=TextureCoordinate(), Time=Constant() {r=3.4})) {constB=-0.5}
  n5 = MaterialFunctionCall(Alpha=ComponentMask(None=Abs(None=@n4)) {r=true}, Min=Constant(), Max=Constant() {r=0.5}) {materialFunction=&SmoothStep}
  n6 = TextureCoordinate()
  n7 = ComponentMask(None=@n6) {r=true}
  n8 = AppendVector(A=LinearInterpolate(A=OneMinus(None=@n7), B=@n7, Alpha=ScalarParameter() {parameterName=bFlipPortrait}), B=ComponentMask(None=@n6) {g=true})
  n9 = ScalarParameter() {defaultValue=0.94, parameterName=ScaleByCenter}
  n10 = ScalarParameter() {parameterName=Position X}
  n11 = ScalarParameter() {parameterName=Position Y}
  n12 = TextureSampleParameter2D(UVs=Add(
    A=MaterialFunctionCall(UVs=@n8, Texture Scale=@n9) {materialFunction=&ScaleUVsByCenter}.UVs
    B=AppendVector(A=@n10, B=@n11)
   )) {parameterName=WorkerPortrait, texture=/Game/UI/Foundation/Textures/Icons/Workers/T-Icon-Leaders-Portrait-Trainer-Male-L}
  n13 = TextureObjectParameter() {
   group=LUT
   parameterName=LUT Volume Texture
   samplerType=SAMPLERTYPE_LinearColor
   sortPriority=50
   texture=/Game/ArtTools/Tech_Art/LUT/LUTs/VT_FE1840_1_inverse
  }
  n14 = StaticSwitchParameter(
   True=MaterialFunctionCall!(In=@n12.RGB, LUT=@n13, LUT Multiplier=ScalarParameter() {defaultValue=1, group=LUT, parameterName=LUT Multiplier, sortPriority=100}) {materialFunction=/Game/Packages/Fortress_SharedMaterials/Anime_Master_Material/MF_ApplyLUT}.Result
   False=@n12.RGB
  ) {group=LUT, parameterName=Use Apply LUT, sortPriority=20}
  n15 = VectorParameter() {parameterName=RGBA1}
  n16 = VectorParameter() {parameterName=RGBA0, defaultValue=<rgba 0.0125 0.0125 0.0125 0>}
  n17 = LinearInterpolate(Alpha=ComponentMask(None=TextureCoordinate()) {g=true}) {constA=0.25}
  n18 = StaticSwitchParameter(True=@n14, False=LinearInterpolate(A=LinearInterpolate(A=@n15.RGB, B=@n16.RGB, Alpha=@n17), B=@n14, Alpha=@n12.A)) {parameterName=IgnoreRarity}
  n19 = ScalarParameter() {defaultValue=1, parameterName=Contrast}
  n20 = MaterialFunctionCall(Float2=TextureCoordinate()) {materialFunction=&BreakOutFloat2Components}
  n21 = Constant() {r=0.5}
  n22 = MaterialFunctionCall(
   Alpha=Max(A=Distance(A=@n20.R, B=@n21), B=Distance(A=@n20.G, B=@n21))
   Min=ScalarParameter() {defaultValue=0.455, parameterName=FrameMin_1}
   Max=ScalarParameter() {defaultValue=0.45, parameterName=FrameMax_1}
  ) {materialFunction=&SmoothStep}
  n23 = Clamp(None=@n1)
  LinearInterpolate(
   A=Multiply(A=LinearInterpolate(A=@n2, B=@n3, Alpha=@n5.Result), B=ScalarParameter() {defaultValue=1, parameterName=BrightnessFrame})
   B=Multiply(A=Power(Base=@n18, Exp=@n19), B=ScalarParameter() {defaultValue=1, parameterName=Brightness})
   Alpha=Clamp(None=Multiply(A=@n22.Result, B=@n23))
  )
  n24 = TextureCoordinate() {uTiling=2, vTiling=2}
  n25 = TextureSample(UVs=@n24) {samplerType=SAMPLERTYPE_Masks, texture=&/S_Generic_Frame_S}
  LinearInterpolate(
   A=ScalarParameter() {defaultValue=0.2, parameterName=RoughnessMin}
   B=ScalarParameter() {defaultValue=0.8, parameterName=RoughnessMax}
   Alpha=@n25.B
  )
  MaterialFunctionCall(Min=ScalarParameter() {defaultValue=0.3, parameterName=FrameMin}, Max=ScalarParameter() {defaultValue=0.5, parameterName=FrameMax}) {materialFunction=&SmoothStep}
  n26 = Constant()
  ShadingPathSwitch(Default=@n25.G, Mobile=@n26) {desc=Migrated from FeatureLevelSwitch}
  ShadingPathSwitch(Default=Constant(), Mobile=@n26) {desc=Migrated from FeatureLevelSwitch}
  TextureSample() {texture=&/T-PhotoMask}
  TextureSample(UVs=@n24) {samplerType=SAMPLERTYPE_Normal, texture=&/S_Generic_Frame_N}
 ]
}
