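# Physics asset for the SK_Banner skeletal mesh, filed with the worker assets for the frontend item inspect UI.
# It gives seven bones one capsule body each, covering the legs, lower and upper staff, central device, flag assembly, flag base and flag.
# Constraints with uniform 45-degree swing and twist limit values chain the legs up through the device to the flag assembly, where flag base and flag branch off.
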PhysicsAsset /Game/UI/Frontend/ItemInspect/Workers/SK_Banner_PhysicsAsset
bodies {
 b_Device [{shapeType=Capsule, length=45.5507, radius=4.5318, center=<-0.0315 0.2032 79.7623>, rotation=<rot 0.2776 -90 -0.0419>}]
 b_Flag [{shapeType=Capsule, length=0.505, radius=0.505, center=<0 0.3689 -41.8567>, rotation=<rot 0 -90 0>}]
 b_Flag_Assembly [{shapeType=Capsule, length=35.276, radius=9.4415, center=<-0.3165 5.9748 -0.6138>, rotation=<rot 0 89.9549 -89.9975>}]
 b_Flag_Base [{shapeType=Capsule, length=26.3812, radius=2.3439, center=<-0.004 0.094 -1.6028>, rotation=<rot 0 -89.9495 -89.852>}]
 b_Leg_Mstr [{shapeType=Capsule, length=20.3379, radius=20.3353, center=<0.0054 -0.0125 6.9429>, rotation=<rot 0.0451 41.4874 -90>}]
 b_LowStaff_05 [{shapeType=Capsule, length=25.9417, radius=4.3654, center=<-0.0448 -0.0329 -6.3642>, rotation=<rot 0.2482 -90 0.4065>}]
 b_UpStaff_02 [{shapeType=Capsule, length=29.6647, radius=5.5618, center=<0.083 0.0756 29.3609>, rotation=<rot -0.0307 -90 -0.174>}]
}
constraints [
 {bone1Name=b_LowStaff_05, bone2Name=b_Device, swing1LimitDegrees=45, swing2LimitDegrees=45, twistLimitDegrees=45}
 {bone1Name=b_Leg_Mstr, bone2Name=b_LowStaff_05, swing1LimitDegrees=45, swing2LimitDegrees=45, twistLimitDegrees=45}
 {bone1Name=b_UpStaff_02, bone2Name=b_Device, swing1LimitDegrees=45, swing2LimitDegrees=45, twistLimitDegrees=45}
 {bone1Name=b_Flag_Assembly, bone2Name=b_UpStaff_02, swing1LimitDegrees=45, swing2LimitDegrees=45, twistLimitDegrees=45}
 {bone1Name=b_Flag_Base, bone2Name=b_Flag_Assembly, swing1LimitDegrees=45, swing2LimitDegrees=45, twistLimitDegrees=45}
 {bone1Name=b_Flag, bone2Name=b_Flag_Assembly, swing1LimitDegrees=45, swing2LimitDegrees=45, twistLimitDegrees=45}
]
